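# Sound cue for aiming down sights with the burst-fire SMG pistol, played as weapon foley.
# A single ADS wave player feeds a Modulator that randomly raises pitch slightly on each play to vary repeats.
# It uses the weapon reload attenuation asset at reduced volume, with overridden concurrency that prevents new instances once the limit is reached.
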SoundCue /Game/Athena/Sounds/Weapons/Pistol_BurstFireSMG/Pistol_BurstFireSMG_ADS_Cue
root=Modulator_0
nodes [
 {id=Modulator_0, children [WavePlayer_1], properties {pitchMax=1.15, pitchMin=1}}
 {id=WavePlayer_1, properties {soundWaveAssetPtr=&/Pistol_BurstFireSMG_ADS_01}}
]
properties {
 attenuationSettings=&Fort_Weapon_Reload_Att
 bOverrideConcurrency=true
 maxDistance=2250
 soundClassObject=&Fort_WeaponFoley
 volumeMultiplier=0.4
 attenuationOverrides {falloffDistance=4100, attenuationShapeExtents {x=900}}
 concurrencyOverrides {resolutionRule=PreventNew}
}
