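# First-person fire sound cue for the Pistol BurstFireSMG weapon, layering close gunshots with delayed bullet-casing drops.
# A mixer blends a randomly chosen, modulated shot with a quieter random casing sound.
# Casings start after a short random delay, with wide volume variation and fixed pitch.
# The cue uses first-person attenuation, a ranged-fire sound class and a gunfire concurrency set, and has its debug flag enabled.
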SoundCue /Game/Athena/Sounds/Weapons/Pistol_BurstFireSMG/Pistol_BurstFireSMG_Fire1P_Cue
root=Mixer_0
nodes [
 {id=Delay_0, children [Random_0], properties {delayMax=0.6, delayMin=0.45}}
 {id=Mixer_0, children [Modulator_2,Modulator_0], properties {inputVolume [1,0.35]}}
 {id=Modulator_0, children [Delay_0], properties {pitchMax=1, pitchMin=1, volumeMax=1, volumeMin=0.05}}
 {id=Modulator_2, children [Random_5], properties {pitchMin=1}}
 {id=Random_0, children [WavePlayer_0,WavePlayer_2,WavePlayer_3,WavePlayer_7,WavePlayer_8], properties {weights [1,1,1,1,1]}}
 {id=Random_5, children [WavePlayer_1,WavePlayer_4,WavePlayer_5,WavePlayer_6], properties {weights [1,1,1,1]}}
 {id=WavePlayer_0, properties {soundWaveAssetPtr=&/Pistol_BurstFireSMG_Bullet_Casings_Drop_05}}
 {id=WavePlayer_1, properties {soundWaveAssetPtr=&/Pistol_BurstFireSMG_Fire_Close_1p_01}}
 {id=WavePlayer_2, properties {soundWaveAssetPtr=&/Pistol_BurstFireSMG_Bullet_Casings_Drop_01}}
 {id=WavePlayer_3, properties {soundWaveAssetPtr=&/Pistol_BurstFireSMG_Bullet_Casings_Drop_02}}
 {id=WavePlayer_4, properties {soundWaveAssetPtr=&/Pistol_BurstFireSMG_Fire_Close_1p_02}}
 {id=WavePlayer_5, properties {soundWaveAssetPtr=&/Pistol_BurstFireSMG_Fire_Close_1p_03}}
 {id=WavePlayer_6, properties {soundWaveAssetPtr=&/Pistol_BurstFireSMG_Fire_Close_1p_04}}
 {id=WavePlayer_7, properties {soundWaveAssetPtr=&/Pistol_BurstFireSMG_Bullet_Casings_Drop_03}}
 {id=WavePlayer_8, properties {soundWaveAssetPtr=&/Pistol_BurstFireSMG_Bullet_Casings_Drop_04}}
]
properties {
 attenuationSettings=&Fort_Weapon_Projectile_FirstPerson_Athena_Att
 bDebug=true
 maxDistance=2097152
 soundClassObject=&Fort_WeaponRanged_Fire_1P_PSM
 volumeMultiplier=0.8
 concurrencyOverrides {bLimitToOwner=true, maxCount=3, resolutionRule=StopOldest}
 concurrencySet [&Weapon_Gunfire_FirstPerson_Conc]
}
