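# First-person fire-end sound cue for the Pistol_BurstFireSMG weapon, played at reduced volume.
# A Random node picks among three equally weighted Hydra three-round-burst stereo end waves, borrowed from the AutoRifle sound folder.
# A Modulator then varies volume and pitch slightly, with pitch never dropping below normal.
# It uses first-person weapon attenuation, a ranged-fire sound class and the shared first-person gunfire concurrency; its debug flag is on.
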SoundCue /Game/Athena/Sounds/Weapons/Pistol_BurstFireSMG/Pistol_BurstFireSMG_FireEnd_1P_Cue
root=Modulator_2
nodes [
 {id=Modulator_2, children [Random_2], properties {pitchMin=1}}
 {id=Random_2, children [WavePlayer_0,WavePlayer_1,WavePlayer_2], properties {weights [1,1,1]}}
 {id=WavePlayer_0, properties {soundWaveAssetPtr=/Game/Sounds/Fort_Weapons/AutoRifle/Weapon_Hydra3RoundBurst_Stereo_End_01}}
 {id=WavePlayer_1, properties {soundWaveAssetPtr=/Game/Sounds/Fort_Weapons/AutoRifle/Weapon_Hydra3RoundBurst_Stereo_End_02}}
 {id=WavePlayer_2, properties {soundWaveAssetPtr=/Game/Sounds/Fort_Weapons/AutoRifle/Weapon_Hydra3RoundBurst_Stereo_End_03}}
]
properties {
 attenuationSettings=&Fort_Weapon_Projectile_FirstPerson_Athena_Att
 bDebug=true
 maxDistance=2097152
 soundClassObject=&Fort_WeaponRanged_Fire_1P_PSM
 volumeMultiplier=0.7
 concurrencyOverrides {bLimitToOwner=true, maxCount=2, resolutionRule=StopOldest}
 concurrencySet [&Weapon_Gunfire_FirstPerson_Conc]
}
