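# Sound cue for the burst-fire SMG pistol's fire-end sound, built from Hydra three-round-burst rifle waves.
# A mixer blends close, medium and distant layers at decreasing volume, each with its own attenuation, modulation and sound class override.
# Close picks randomly from three waves; medium and distant switch by quality level between random variants and one optimized wave.
# The medium random pool borrows two distant waves, and owner-limited concurrency stops the oldest instance.
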SoundCue /Game/Athena/Sounds/Weapons/Pistol_BurstFireSMG/Pistol_BurstFireSMG_FireEnd_Cue
root=Mixer_0
nodes [
 {id=Attenuation_2, children [Modulator_2], properties {attenuationSettings=&Fort_Weapon_Projectile_Close_Athena_Att}}
 {
  id=Attenuation_3
  children [Modulator_3]
  properties {attenuationSettings=&Fort_Weapon_Projectile_Medium_Mono_Athena_Att}
 }
 {
  id=Attenuation_4
  children [Modulator_4]
  properties {attenuationSettings=&Fort_Weapon_Projectile_Distant_Mono_Athena_Att}
 }
 {id=Mixer_0, children [SoundClass_2,SoundClass_4,SoundClass_3], properties {inputVolume [1,0.6,0.3]}}
 {id=Modulator_2, children [Random_2], properties {pitchMin=1}}
 {id=Modulator_3, children [QualityLevel_1], properties {volumeMax=2.05, volumeMin=1.95}}
 {id=Modulator_4, children [QualityLevel_0], properties {volumeMax=2.05, volumeMin=1.95}}
 {id=QualityLevel_0, children [Random_3,Random_3,WavePlayer_9,WavePlayer_9]}
 {id=QualityLevel_1, children [Random_4, Random_4, WavePlayer_10, WavePlayer_10]}
 {id=Random_2, children [WavePlayer_0,WavePlayer_1,WavePlayer_2], properties {weights [1,1,1]}}
 {id=Random_3, children [WavePlayer_3,WavePlayer_4,WavePlayer_5], properties {weights [1,1,1]}}
 {id=Random_4, children [WavePlayer_6,WavePlayer_7,WavePlayer_8], properties {weights [1,1,1]}}
 {
  id=SoundClass_2
  children [Attenuation_2]
  properties {soundClassOverride=&Fort_WeaponRanged_Fire_3P_PSM}
 }
 {
  id=SoundClass_3
  children [Attenuation_4]
  properties {soundClassOverride=&Fort_WeaponRanged_Fire_Distant}
 }
 {
  id=SoundClass_4
  children [Attenuation_3]
  properties {soundClassOverride=&Fort_WeaponRanged_Fire_3P_PSM}
 }
 {id=WavePlayer_0, properties {soundWaveAssetPtr=/Game/Sounds/Fort_Weapons/AutoRifle/Weapon_Hydra3RoundBurst_Stereo_End_01}}
 {id=WavePlayer_1, properties {soundWaveAssetPtr=/Game/Sounds/Fort_Weapons/AutoRifle/Weapon_Hydra3RoundBurst_Stereo_End_02}}
 {id=WavePlayer_10, properties {soundWaveAssetPtr=/Game/Sounds/Fort_Weapons/AutoRifle/Weapon_Hydra3RoundBurst_MediumNewFire_Stereo_End_opt}}
 {id=WavePlayer_2, properties {soundWaveAssetPtr=/Game/Sounds/Fort_Weapons/AutoRifle/Weapon_Hydra3RoundBurst_Stereo_End_03}}
 {id=WavePlayer_3, properties {soundWaveAssetPtr=/Game/Sounds/Fort_Weapons/AutoRifle/Weapon_Hydra3RoundBurst_DistantNewFire_Stereo_End_01}}
 {id=WavePlayer_4, properties {soundWaveAssetPtr=/Game/Sounds/Fort_Weapons/AutoRifle/Weapon_Hydra3RoundBurst_DistantNewFire_Stereo_End_02}}
 {id=WavePlayer_5, properties {soundWaveAssetPtr=/Game/Sounds/Fort_Weapons/AutoRifle/Weapon_Hydra3RoundBurst_DistantNewFire_Stereo_End_03}}
 {id=WavePlayer_6, properties {soundWaveAssetPtr=&Weapon_Hydra3RoundBurst_DistantNewFire_Stereo_End_03}}
 {id=WavePlayer_7, properties {soundWaveAssetPtr=/Game/Sounds/Fort_Weapons/AutoRifle/Weapon_Hydra3RoundBurst_MediumNewFire_Stereo_End_03}}
 {id=WavePlayer_8, properties {soundWaveAssetPtr=&Weapon_Hydra3RoundBurst_DistantNewFire_Stereo_End_02}}
 {id=WavePlayer_9, properties {soundWaveAssetPtr=/Game/Sounds/Fort_Weapons/AutoRifle/Weapon_Hydra3RoundBurst_DistantNewFire_Stereo_End_opt}}
]
properties {
 bDebug=true
 maxDistance=27500
 soundClassObject=&Fort_WeaponRanged_Fire_3P_PSM
 volumeMultiplier=0.7
 concurrencyOverrides {bLimitToOwner=true, maxCount=2, resolutionRule=StopOldest}
 concurrencySet [&Weapon_Gunfire_Global_Conc, &Weapon_Gunfire_Local_Conc]
}
