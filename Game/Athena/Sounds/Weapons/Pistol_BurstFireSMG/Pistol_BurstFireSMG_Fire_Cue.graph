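# Sound cue for the third-person gunfire of the burst-fire SMG pistol in Athena.
# A mixer layers close, medium and distant branches, with the close layer weighted loudest.
# Each branch randomly picks one of four waves, modulates it, then applies its own distance attenuation and sound class override, with medium and distant sharing the Distant class.
# Owner-limited, stop-oldest concurrency settings sit alongside global and local gunfire concurrency sets.
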SoundCue /Game/Athena/Sounds/Weapons/Pistol_BurstFireSMG/Pistol_BurstFireSMG_Fire_Cue
root=Mixer_0
nodes [
 {id=Attenuation_2, children [Modulator_2], properties {attenuationSettings=&Fort_Weapon_Projectile_Close_Athena_Att}}
 {
  id=Attenuation_3
  children [Modulator_3]
  properties {attenuationSettings=&Fort_Weapon_Projectile_Medium_Mono_Athena_Att}
 }
 {
  id=Attenuation_4
  children [Modulator_4]
  properties {attenuationSettings=&Fort_Weapon_Projectile_Distant_Mono_Athena_Att}
 }
 {id=Mixer_0, children [SoundClass_2,SoundClass_4,SoundClass_3], properties {inputVolume [1.2,0.5,0.4]}}
 {id=Modulator_2, children [Random_5], properties {pitchMin=1}}
 {id=Modulator_3, children [Random_2], properties {volumeMax=2.05, volumeMin=1.95}}
 {id=Modulator_4, children [Random_3], properties {volumeMax=2.05, volumeMin=1.95}}
 {id=Random_2, children [WavePlayer_10, WavePlayer_11, WavePlayer_12, WavePlayer_13], properties {weights [1,1,1,1]}}
 {id=Random_3, children [WavePlayer_14, WavePlayer_18, WavePlayer_19, WavePlayer_20], properties {weights [1,1,1,1]}}
 {id=Random_5, children [WavePlayer_6,WavePlayer_7,WavePlayer_8,WavePlayer_9], properties {weights [1,1,1,1]}}
 {
  id=SoundClass_2
  children [Attenuation_2]
  properties {soundClassOverride=&Fort_WeaponRanged_Fire_3P_PSM}
 }
 {
  id=SoundClass_3
  children [Attenuation_4]
  properties {soundClassOverride=&Fort_WeaponRanged_Fire_Distant}
 }
 {
  id=SoundClass_4
  children [Attenuation_3]
  properties {soundClassOverride=&Fort_WeaponRanged_Fire_Distant}
 }
 {id=WavePlayer_10, properties {soundWaveAssetPtr=&/Pistol_BurstFireSMG_Fire_Med_01}}
 {id=WavePlayer_11, properties {soundWaveAssetPtr=&/Pistol_BurstFireSMG_Fire_Med_02}}
 {id=WavePlayer_12, properties {soundWaveAssetPtr=&/Pistol_BurstFireSMG_Fire_Med_03}}
 {id=WavePlayer_13, properties {soundWaveAssetPtr=&/Pistol_BurstFireSMG_Fire_Med_04}}
 {id=WavePlayer_14, properties {soundWaveAssetPtr=&/Pistol_BurstFireSMG_Fire_Distant_01}}
 {id=WavePlayer_18, properties {soundWaveAssetPtr=&/Pistol_BurstFireSMG_Fire_Distant_02}}
 {id=WavePlayer_19, properties {soundWaveAssetPtr=&/Pistol_BurstFireSMG_Fire_Distant_03}}
 {id=WavePlayer_20, properties {soundWaveAssetPtr=&/Pistol_BurstFireSMG_Fire_Distant_04}}
 {id=WavePlayer_6, properties {soundWaveAssetPtr=&/Pistol_BurstFireSMG_Fire_Close_3p_01}}
 {id=WavePlayer_7, properties {soundWaveAssetPtr=&/Pistol_BurstFireSMG_Fire_Close_3p_02}}
 {id=WavePlayer_8, properties {soundWaveAssetPtr=&/Pistol_BurstFireSMG_Fire_Close_3p_03}}
 {id=WavePlayer_9, properties {soundWaveAssetPtr=&/Pistol_BurstFireSMG_Fire_Close_3p_04}}
]
properties {
 bDebug=true
 maxDistance=27500
 soundClassObject=&Fort_WeaponRanged_Fire_3P_PSM
 volumeMultiplier=0.8
 concurrencyOverrides {bLimitToOwner=true, maxCount=3, resolutionRule=StopOldest}
 concurrencySet [&Weapon_Gunfire_Global_Conc, &Weapon_Gunfire_Local_Conc]
}
