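# Sound cue for the low-ammo sound of the burst-fire SMG pistol weapon.
# A single wave player feeds the LowAmmo wave straight to the output, with raised pitch and volume.
# It uses a degraded ranged-weapon low-ammo sound class and overrides concurrency to prevent new instances once the limit is reached.
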SoundCue /Game/Athena/Sounds/Weapons/Pistol_BurstFireSMG/Pistol_BurstFireSMG_Low_Ammo_Cue
root=WavePlayer_0
nodes [{id=WavePlayer_0, properties {soundWaveAssetPtr=&/Pistol_BurstFireSMG_LowAmmo}}]
properties {
 bOverrideConcurrency=true
 maxDistance=2097152
 pitchMultiplier=1.2
 soundClassObject=/Game/Sounds/FortSoundClassesAndModes/Fort_WeaponRanged_LowAmmo_Degraded_PSM
 volumeMultiplier=1.5
 concurrencyOverrides {resolutionRule=PreventNew}
}
