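# Sound cue for picking up the burst-fire SMG pistol weapon in Athena.
# It plays a single pickup wave directly, with no randomization or modulation nodes.
# Weapon pickup attenuation and the loot collect sound class shape its spatialization and mix, at reduced volume.
# Overridden concurrency allows only one instance at a time and blocks new ones while it plays.
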SoundCue /Game/Athena/Sounds/Weapons/Pistol_BurstFireSMG/Pistol_BurstFireSMG_Pickup_Cue
root=WavePlayer_1
nodes [{id=WavePlayer_1, properties {soundWaveAssetPtr=&/Pistol_BurstFireSMG_Pickup_02}}]
properties {
 attenuationSettings=&Fort_Weapon_Pickup_Att
 bOverrideConcurrency=true
 maxDistance=3500
 soundClassObject=&Fort_Loot_Collect
 volumeMultiplier=0.6923
 concurrencyOverrides {maxCount=1, resolutionRule=PreventNew}
}
