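# Sound cue for the end of the reload on the burst-fire SMG pistol in Athena, classed as weapon foley.
# A single wave player feeds the output directly, with no randomization or mixing, at reduced volume.
# Distance falloff uses the shared weapon reload attenuation asset.
# The cue overrides concurrency with a prevent-new rule, so new plays are refused once the voice limit is reached.
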SoundCue /Game/Athena/Sounds/Weapons/Pistol_BurstFireSMG/Pistol_BurstFireSMG_Reload_End_Cue
root=WavePlayer_1
nodes [{id=WavePlayer_1, properties {soundWaveAssetPtr=&/Pistol_BurstFireSMG_Reload_End_03}}]
properties {
 attenuationSettings=&Fort_Weapon_Reload_Att
 bOverrideConcurrency=true
 maxDistance=2250
 soundClassObject=&Fort_WeaponFoley
 volumeMultiplier=0.64
 attenuationOverrides {falloffDistance=4100, attenuationShapeExtents {x=900}}
 concurrencyOverrides {resolutionRule=PreventNew}
}
