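# Sound cue for the magazine-insert step of the burst-fire SMG pistol reload.
# A single mid-reload wave plays through a Modulator that locks pitch slightly lower, with no random pitch variation.
# The cue plays at reduced volume in the weapon foley sound class, using the shared weapon reload attenuation.
# Its own concurrency override refuses new instances once the limit is reached instead of stopping playing ones.
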SoundCue /Game/Athena/Sounds/Weapons/Pistol_BurstFireSMG/Pistol_BurstFireSMG_Reload_Insert_Cue
root=Modulator_1
nodes [
 {id=Modulator_1, children [WavePlayer_0], properties {pitchMax=0.9, pitchMin=0.9}}
 {id=WavePlayer_0, properties {soundWaveAssetPtr=&/Pistol_BurstFireSMG_Reload_Mid_02}}
]
properties {
 attenuationSettings=&Fort_Weapon_Reload_Att
 bOverrideConcurrency=true
 maxDistance=2250
 soundClassObject=&Fort_WeaponFoley
 volumeMultiplier=0.5
 attenuationOverrides {falloffDistance=4100, attenuationShapeExtents {x=900}}
 concurrencyOverrides {resolutionRule=PreventNew}
}
